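# Sound cue for first-person crouch-sprint footsteps on the Cube surface in Athena, playing Cubetown run footstep waves.
# A quality switch sends the first two levels to a random pick of ten waves and the last two to a random pick of four, each through a default modulator.
# It uses crouch-sprint attenuation, the 1P foley sound class and owner-limited footstep concurrency.
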SoundCue /Game/Athena/Sounds/Character/Footsteps/1P/Cube/Athena_Soldier_Cube_Crouch_Sprint_Footstep_Cue
root=QualityLevel_0
nodes [
 {id=Modulator_0, children [Random_1]}
 {id=Modulator_1, children [Random_0]}
 {id=QualityLevel_0, children [Modulator_1,Modulator_1,Modulator_0,Modulator_0]}
 {
  id=Random_0
  children [
   WavePlayer_0
   WavePlayer_4
   WavePlayer_6
   WavePlayer_7
   WavePlayer_8
   WavePlayer_9
   WavePlayer_10
   WavePlayer_11
   WavePlayer_12
   WavePlayer_13
  ]
  properties {weights [1,1,1,1,1,1,1,1,1,1]}
 }
 {
  id=Random_1
  children [WavePlayer_14, WavePlayer_15, WavePlayer_16, WavePlayer_17]
  properties {bShouldExcludeFromBranchCulling=true, weights [1,1,1,1]}
 }
 {id=WavePlayer_0, properties {soundWaveAssetPtr=&/FS_Cubetown_Run_01}}
 {id=WavePlayer_10, properties {soundWaveAssetPtr=&/FS_Cubetown_Run_07}}
 {id=WavePlayer_11, properties {soundWaveAssetPtr=&/FS_Cubetown_Run_08}}
 {id=WavePlayer_12, properties {soundWaveAssetPtr=&/FS_Cubetown_Run_09}}
 {id=WavePlayer_13, properties {soundWaveAssetPtr=&/FS_Cubetown_Run_10}}
 {id=WavePlayer_14, properties {soundWaveAssetPtr=&FS_Cubetown_Run_01}}
 {id=WavePlayer_15, properties {soundWaveAssetPtr=&/FS_Cubetown_Run_02}}
 {id=WavePlayer_16, properties {soundWaveAssetPtr=&/FS_Cubetown_Run_03}}
 {id=WavePlayer_17, properties {soundWaveAssetPtr=&/FS_Cubetown_Run_04}}
 {id=WavePlayer_4, properties {soundWaveAssetPtr=&FS_Cubetown_Run_02}}
 {id=WavePlayer_6, properties {soundWaveAssetPtr=&FS_Cubetown_Run_03}}
 {id=WavePlayer_7, properties {soundWaveAssetPtr=&FS_Cubetown_Run_04}}
 {id=WavePlayer_8, properties {soundWaveAssetPtr=&/FS_Cubetown_Run_05}}
 {id=WavePlayer_9, properties {soundWaveAssetPtr=&/FS_Cubetown_Run_06}}
]
properties {
 attenuationSettings=/Game/Sounds/Attenuation/Fort_Player_Footsteps_1P_CrouchSprint_Athena_Att
 maxDistance=2097152
 soundClassObject=&Fort_Player_Foley_1P_Athena
 volumeMultiplier=0.4
 concurrencyOverrides {bLimitToOwner=true, maxCount=1, resolutionRule=StopOldest}
 concurrencySet [&Athena_Footstep_Local_Conc]
}
