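# First-person walk footstep Sound Cue for Athena soldiers on the Cube (Cubetown) surface.
# Each branch mixes a random Cubetown step with quieter Commando foley, the foley given wider pitch and volume variation.
# A quality node gives the first two levels the full sample pools and the last two smaller ones.
# Owner-limited concurrency allows one instance, stopping the oldest, with 1P walk attenuation.
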SoundCue /Game/Athena/Sounds/Character/Footsteps/1P/Cube/Athena_Soldier_Cube_Walk_Footstep_Cue
root=QualityLevel_0
nodes [
 {id=Mixer_0, children [Modulator_1,Modulator_2], properties {inputVolume [0.1,1]}}
 {id=Mixer_1, children [Modulator_11,Modulator_0], properties {inputVolume [0.1,1]}}
 {id=Modulator_0, children [Random_4]}
 {id=Modulator_1, children [Random_2], properties {pitchMax=1, pitchMin=0.7, volumeMax=1, volumeMin=0.3}}
 {id=Modulator_11, children [Random_3], properties {pitchMax=1, pitchMin=0.7, volumeMax=1, volumeMin=0.3}}
 {id=Modulator_2, children [Random_1]}
 {id=QualityLevel_0, children [Mixer_1,Mixer_1,Mixer_0,Mixer_0]}
 {id=Random_1, children [WavePlayer_33, WavePlayer_34, WavePlayer_35], properties {bShouldExcludeFromBranchCulling=true, weights [1,1,1]}}
 {id=Random_2, children [WavePlayer_23, WavePlayer_24], properties {bShouldExcludeFromBranchCulling=true, weights [1,1]}}
 {
  id=Random_3
  children [WavePlayer_6, WavePlayer_7, WavePlayer_8, WavePlayer_9, WavePlayer_10, WavePlayer_11]
  properties {bShouldExcludeFromBranchCulling=true, weights [1,1,1,1,1,1]}
 }
 {
  id=Random_4
  children [
   WavePlayer_20
   WavePlayer_21
   WavePlayer_25
   WavePlayer_26
   WavePlayer_27
   WavePlayer_28
   WavePlayer_29
   WavePlayer_30
   WavePlayer_31
   WavePlayer_32
  ]
  properties {weights [1,1,1,1,1,1,1,1,1,1]}
 }
 {id=WavePlayer_10, properties {soundWaveAssetPtr=/Game/Sounds/Fortress_Foley/Commando/Commando_Foley_Stereo_005}}
 {id=WavePlayer_11, properties {soundWaveAssetPtr=/Game/Sounds/Fortress_Foley/Commando/Commando_Foley_Stereo_009}}
 {id=WavePlayer_20, properties {soundWaveAssetPtr=&/FS_Cubetown_Walk_01}}
 {id=WavePlayer_21, properties {soundWaveAssetPtr=&/FS_Cubetown_Walk_02}}
 {id=WavePlayer_23, properties {soundWaveAssetPtr=/Game/Sounds/Fortress_Foley/Commando/Commando_Foley_Stereo_006}}
 {id=WavePlayer_24, properties {soundWaveAssetPtr=/Game/Sounds/Fortress_Foley/Commando/Commando_Foley_Stereo_008}}
 {id=WavePlayer_25, properties {soundWaveAssetPtr=&/FS_Cubetown_Walk_03}}
 {id=WavePlayer_26, properties {soundWaveAssetPtr=&/FS_Cubetown_Walk_04}}
 {id=WavePlayer_27, properties {soundWaveAssetPtr=&/FS_Cubetown_Walk_05}}
 {id=WavePlayer_28, properties {soundWaveAssetPtr=&/FS_Cubetown_Walk_06}}
 {id=WavePlayer_29, properties {soundWaveAssetPtr=&/FS_Cubetown_Walk_07}}
 {id=WavePlayer_30, properties {soundWaveAssetPtr=&/FS_Cubetown_Walk_08}}
 {id=WavePlayer_31, properties {soundWaveAssetPtr=&/FS_Cubetown_Walk_09}}
 {id=WavePlayer_32, properties {soundWaveAssetPtr=&/FS_Cubetown_Walk_10}}
 {id=WavePlayer_33, properties {soundWaveAssetPtr=&FS_Cubetown_Walk_01}}
 {id=WavePlayer_34, properties {soundWaveAssetPtr=&FS_Cubetown_Walk_02}}
 {id=WavePlayer_35, properties {soundWaveAssetPtr=&FS_Cubetown_Walk_03}}
 {id=WavePlayer_6, properties {soundWaveAssetPtr=&Commando_Foley_Stereo_006}}
 {id=WavePlayer_7, properties {soundWaveAssetPtr=&Commando_Foley_Stereo_008}}
 {id=WavePlayer_8, properties {soundWaveAssetPtr=/Game/Sounds/Fortress_Foley/Commando/Commando_Foley_Stereo_001}}
 {id=WavePlayer_9, properties {soundWaveAssetPtr=/Game/Sounds/Fortress_Foley/Commando/Commando_Foley_Stereo_003}}
]
properties {
 attenuationSettings=/Game/Sounds/Attenuation/Fort_Player_Footsteps_1P_Walk_Athena_Att
 maxDistance=2097152
 soundClassObject=&Fort_Player_Foley_1P_Athena
 volumeMultiplier=0.5
 concurrencyOverrides {bLimitToOwner=true, maxCount=1, resolutionRule=StopOldest}
 concurrencySet [&Athena_Footstep_Local_Conc]
}
